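; Burst gameplay cue notify for the temple ruins item door, played when the door refuses an item, that colors its effect by the item rarity the door requires.
; On burst it takes the first spawned FX component, casts it to Niagara, and casts the target to the temple door Blueprint.
; It copies the door's required rarity integer into the "ColorFX" Niagara variable, then calls the parent handler, only when that component is valid.
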(fn OnBurst(EitherVFX-preferthisversion)
  (params (MyTarget :type Actor%20Object%20Reference) (Parameters :type Gameplay%20Cue%20Parameters%20Structure%20%28by%20ref%29) (ParticleComponents :type Array%20of%20FXSystem%20Component%20Object%20References) (AudioComponents :type Array%20of%20Audio%20Component%20Object%20References) (BurstCameraShakeInstance :type Camera%20Shake%20Base%20Object%20Reference) (BurstDecalInstance :type Decal%20Actor%20Object%20Reference))
  (bind _output (Utilities|Array|Get(acopy) ParticleComponents))
  ;[1 1 1] Grab the first index 0, as we only have 1 FX in here, and map it to the color of the door rarity to open
  (bind _asbp_chrono_requires_item_temple_door (Utilities|Casting|CastToBP_Chrono_RequiresItem_TempleDoor MyTarget))
  (Utilities|IsValid _output
    (:Is%20Valid
      (bind _asniagara_particle_system_component (Utilities|Casting|CastToNiagaraComponent _output))
      (Niagara|SetNiagaraVariable(Int32) _asniagara_particle_system_component "ColorFX" (Variables|/ChronoEnviroItems/TemplePuzzles/ItemCheck/Gameplay/Actors/BP_Chrono_RequiresItem_TempleDoor.BP_Chrono_RequiresItem_TempleDoor_C|GetRarity%20Check%20Required%20Int _asbp_chrono_requires_item_temple_door))
      (|CallParentFunction MyTarget Parameters ParticleComponents AudioComponents BurstCameraShakeInstance BurstDecalInstance)
      (return))))
