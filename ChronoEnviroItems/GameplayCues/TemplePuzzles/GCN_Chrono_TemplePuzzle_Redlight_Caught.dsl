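; Burst gameplay cue notify that plays the "caught" sound for the Chrono temple red light puzzle.
; The target is cast to a Fort pawn, and a locally viewed pawn gets the first-person 2D sound.
; Otherwise the same target is cast to the PortalCheck statue to play the third-person sound at its mesh location.
; That statue cast sits inside the successful pawn cast, and failed casts play nothing.
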(fn OnBurst(EitherVFX-preferthisversion)
  (params (MyTarget :type Actor%20Object%20Reference) (Parameters :type Gameplay%20Cue%20Parameters%20Structure%20%28by%20ref%29) (ParticleComponents :type Array%20of%20FXSystem%20Component%20Object%20References) (AudioComponents :type Array%20of%20Audio%20Component%20Object%20References) (BurstCameraShakeInstance :type Camera%20Shake%20Base%20Object%20Reference) (BurstDecalInstance :type Decal%20Actor%20Object%20Reference))
  (bind _asfort_pawn (Utilities|Casting|CastToFortPawn MyTarget)
    (:then
      (if (Pawn|IsLocallyViewed _asfort_pawn)
        (Audio|SpawnSound2D MyTarget (Variables|Default|Get1PCaughtSound))
        (else
          (bind _asb_chrono_portal_check_statue (Utilities|Casting|CastToB_Chrono_PortalCheck_Statue MyTarget)
            (:then
              (Audio|SpawnSoundatLocation 0 (Variables|Default|Get3PCaughtSound) (Transformation|GetWorldLocation (Variables|/ChronoEnviroItems/TemplePuzzles/PortalCheck/Actors/B_Chrono_PortalCheck_Statue.B_Chrono_PortalCheck_Statue_C|GetStaticMesh _asb_chrono_portal_check_statue))))
            (:CastFailed
              _)))))
    (:CastFailed
      _))
  (return))
